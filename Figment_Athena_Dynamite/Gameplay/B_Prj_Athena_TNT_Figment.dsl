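; Thrown TNT (dynamite) projectile Blueprint for Figment Athena, adding vehicle bounce behavior on top of its parent class.
; On hit, it casts the other actor to FortVehicleInterface and rewrites the projectile movement velocity so it bounces off vehicles.
; The new velocity is the current velocity scaled up, plus an upward push proportional to current speed.
; BeginPlay only calls the parent, while overlap and tick events are empty.
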(event EventBeginPlay
  (|CallParentFunction))

(event Collision|EventActorBeginOverlap
  (params (OtherActor :type Actor%20Object%20Reference)))

(event EventTick
  (params (DeltaSeconds :type Float%20%28single-precision%29)))

(event Collision|EventHit
  (params (MyComp :type Primitive%20Component%20Object%20Reference) (Other :type Actor%20Object%20Reference) (OtherComp :type Primitive%20Component%20Object%20Reference) (bSelfMoved :type Boolean) (HitLocation :type Vector :default "0, 0, 0") (HitNormal :type Vector :default "0, 0, 0") (NormalImpulse :type Vector :default "0, 0, 0") (Hit :type Hit%20Result%20Structure%20%28by%20ref%29))
  ; Event Hit:
  ; Making sure the projectile bounces off of vehicles
  (Utilities|Casting|CastToFortVehicleInterface Other)
  (Variables|/Script/Engine.MovementComponent|SetVelocity (Math|Vector|vector+vector (* (Variables|/Script/Engine.MovementComponent|GetVelocity (Variables|Movement|GetProjectileMovementComponent)) 2.75) (Math|Vector|MakeVector 0.0 0.0 (* (Math|Vector|VectorLength (Variables|/Script/Engine.MovementComponent|GetVelocity (Variables|Movement|GetProjectileMovementComponent))) 0.8))) (Variables|Movement|GetProjectileMovementComponent)))

(fn ConstructionScript
  (|CallParentFunction))
