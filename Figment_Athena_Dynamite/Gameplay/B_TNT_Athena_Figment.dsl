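; Figment Athena dynamite (TNT) weapon Blueprint that supplies the point locations for a spline along its fuse.
; SplinePointLocations reads four weapon-mesh sockets in order from fuse begin, through two intermediate fuse sockets, to fuse end.
; It converts each socket location into the Spline component's local space and returns them as an array.
; The construction script only calls the parent, and nothing in this Blueprint applies the points to the spline.
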(fn SplinePointLocations
  (return (Utilities|Array|MakeArray (Math|Transform|InverseTransformLocation (Transformation|GetWorldTransform (Variables|Default|GetSpline)) (Transformation|GetSocketLocation (Visual|GetWeaponMesh) "FX_FuseBegin")) (Math|Transform|InverseTransformLocation (Transformation|GetWorldTransform (Variables|Default|GetSpline)) (Transformation|GetSocketLocation (Visual|GetWeaponMesh) "dyn_fuse2")) (Math|Transform|InverseTransformLocation (Transformation|GetWorldTransform (Variables|Default|GetSpline)) (Transformation|GetSocketLocation (Visual|GetWeaponMesh) "dyn_fuse1")) (Math|Transform|InverseTransformLocation (Transformation|GetWorldTransform (Variables|Default|GetSpline)) (Transformation|GetSocketLocation (Visual|GetWeaponMesh) "FX_FuseEnd")))))

(fn ConstructionScript
  (|CallParentFunction))
